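; Parent Blueprint that streams special event video through a media player and sound component wherever cosmetic events can run.
; It requests a URL from the Epic stream media source's VUID, then opens the stream if autostart is pending or playback is running.
; Media callbacks broadcast playback start/end, toggle share blocking, and retry failures with exponential backoff.
; App lifecycle delegates optionally close the stream on backgrounding and reopen it on foregrounding.
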(event EventEndPlay
  (params (EndPlayReason :type EEndPlayReason%20Enum :default "Destroyed"))
  (bind _mediasource (Variables|Default|GetMediaSource))
  (bind _videoplayer (Variables|Default|GetVideoPlayer))
  ;[0 1 0 0.5] End Play
  (CallFunction|CloseSpecialEventStream)
  ;;[0 0 0 0.5] Unbind from media player events
  (Utilities|IsValid _videoplayer
    (:Is%20Valid
      (Media|MediaPlayer|UnbindEventfromOnMediaOpened _videoplayer (EventDispatchers|CreateEvent :!event Custom|OnMediaOpened))
      (Media|MediaPlayer|UnbindEventfromOnMediaOpenFailed _videoplayer (EventDispatchers|CreateEvent :!event Custom|OnMediaOpenedFailed))
      (Media|MediaPlayer|UnbindEventfromOnEndReached _videoplayer (EventDispatchers|CreateEvent :!event Custom|OnMediaEndReached))
      ;;[0 0 0 0.5] Unbind from URL request events
      (Utilities|IsValid _mediasource
        (:Is%20Valid
          (Stream|UnbindEventfromOnVideoUrlSuccess _mediasource (EventDispatchers|CreateEvent :!event OnStreamURLRequestSuccess))
          (Stream|UnbindEventfromOnVideoUrlFailed _mediasource (EventDispatchers|CreateEvent :!event OnStreamURLRequestFailed)))
        (:Is%20Not%20Valid
          (Fortnite|FortShippingLog self "EndPlay: VideoPlayer is not valid!" true))))
    (:Is%20Not%20Valid
      (Fortnite|FortShippingLog self "EndPlay: VideoPlayer is not valid!" true))))

(event EventTick
  (params (DeltaSeconds :type Float%20%28single-precision%29))
  (bind _videoplayer (Variables|Default|GetVideoPlayer))
  ;[0 1 0 0.5] Tick - only active when stream is running
  (Utilities|IsValid _videoplayer
    (:Is%20Valid
      (Development|PrintText (Utilities|Text|FormatText "NSLOCTEXT(\"[EB26832506F51E826F44DB7912CE0A62]\", \"37CADCFF47C0E3A7A7FA879959178194\", \"{Label}: {Timespan}\")" (Utilities|Text|MakeLiteralText "NSLOCTEXT(\"[EB26832506F51E826F44DB7912CE0A62]\", \"651A1E674DCD5D80C7F195BCC44114D7\", \"Playback Timespan\")") (TimeOfDay|GetTimespanasText (Media|MediaPlayer|GetTime _videoplayer) true true "ColonsWithMilliseconds")) true false "(R=0.000000,G=0.660000,B=1.000000,A=1.000000)" 0.0)
      (Development|PrintText (Utilities|Text|FormatText "NSLOCTEXT(\"[EB26832506F51E826F44DB7912CE0A62]\", \"ABFBC5D64E06C8E60BC1238F4FF0AA87\", \"Share Blocking Active: {Active}\")" (Variables|ShareBlocking|GetIsSharingBlocked)) true false "(R=0.000000,G=0.660000,B=1.000000,A=1.000000)" 0.0))))

(event EventBeginPlay
  ;[0 1 0 0.5] Begin Play
  (Networking|CanExecuteCosmeticEvents
    (:True
      (CallFunction|InitMediaFramework))))

(event Custom|InitMediaFramework
  ;[0 0 0 0.5] Media Framework
  ;;[0 0 0 0.5] Setup media player and stream media source
  (Networking|CanExecuteCosmeticEvents
    (:True
      (Fortnite|FortShippingLog self "InitMediaFramework: called...")
      (if (and (Utilities|IsValid (Variables|Default|GetMediaSource) :!class /Script/Engine.KismetSystemLibrary) (Utilities|IsValid (Variables|Default|GetVideoPlayer) :!class /Script/Engine.KismetSystemLibrary))
        (CallFunction|BindMediaOpenCloseEvents)
        (Media|MediaSoundComponent|SetMediaPlayer (Variables|Default|GetMediaSound) (Variables|Default|GetVideoPlayer))
        (if (Variables|Default|GetAutomaticallyRequestVideoURL)
          (CallFunction|TryLoadStreamURL)
          (else
            ;;; Not requesting stream URL due to settings
            (Fortnite|FortShippingLog self "InitMediaFramework: Not requesting the stream URL due to AutomaticallyRequestVideoURL" true)))
        (else
          ;;;[1 0 0] Error: Video Player or Media Source is not set
          (Fortnite|FortShippingLog self "InitMediaFramework: Video Player or Media Source is invalid!" true))))))

(event ApplicationWillEnterBackgroundDelegate(ApplicationLifecycle)
  ;[0 0 0 0.5] Application Lifecycle
  ;; ApplicationWillEnterBackground: Close the stream when the application is pushed to the background to prevent stream speed up upon returning
  (Networking|CanExecuteCosmeticEvents
    (:True
      (if (Variables|ApplicationLifecycle|GetAllowStreamResponseToApplicationLifecycle)
        (Fortnite|FortShippingLog self "ApplicationWillEnterBackground: Closing stream due to application about to enter background")
        (Variables|ApplicationLifecycle|SetStreamStoppedDueToSuspend true)
        (CallFunction|CloseSpecialEventStream)
        (else
          (Fortnite|FortShippingLog self "ApplicationWillEnterBackground: doing nothing - AllowStreamResponseToApplicationLifecycle is false"))))))

(event Custom|TestApplicationWillEnterBackground
  ;[0 0 0 0.5] Application Lifecycle
  ;; ApplicationWillEnterBackground: Close the stream when the application is pushed to the background to prevent stream speed up upon returning
  (Networking|CanExecuteCosmeticEvents
    (:True
      (if (Variables|ApplicationLifecycle|GetAllowStreamResponseToApplicationLifecycle)
        (Fortnite|FortShippingLog self "ApplicationWillEnterBackground: Closing stream due to application about to enter background")
        (Variables|ApplicationLifecycle|SetStreamStoppedDueToSuspend true)
        (CallFunction|CloseSpecialEventStream)
        (else
          (Fortnite|FortShippingLog self "ApplicationWillEnterBackground: doing nothing - AllowStreamResponseToApplicationLifecycle is false"))))))

(event ApplicationHasEnteredForegroundDelegate(ApplicationLifecycle)
  ;[0 0 0 0.5] Application Lifecycle
  ;; ApplicationHasEnteredForeground: Open the stream if we were closed via ApplicationWillEnterBackground delegate
  (Networking|CanExecuteCosmeticEvents
    (:True
      (if (Variables|ApplicationLifecycle|GetAllowStreamResponseToApplicationLifecycle)
        (if (Variables|ApplicationLifecycle|GetStreamStoppedDueToSuspend)
          (Fortnite|FortShippingLog self "ApplicationHasEnteredForeground: Opening stream due to application entering foreground")
          (CallFunction|OpenSpecialEventStream)
          (Variables|ApplicationLifecycle|SetStreamStoppedDueToSuspend false)
          (else
            (Fortnite|FortShippingLog self "ApplicationHasEnteredForeground: Opening stream due to application entering foreground")))
        (else
          (Fortnite|FortShippingLog self "ApplicationHasEnteredForeground: doing nothing - 
AllowStreamResponseToApplicationLifecycle is false"))))))

(event Custom|TestApplicationHasEnteredForeground
  ;[0 0 0 0.5] Application Lifecycle
  ;; ApplicationHasEnteredForeground: Open the stream if we were closed via ApplicationWillEnterBackground delegate
  (Networking|CanExecuteCosmeticEvents
    (:True
      (if (Variables|ApplicationLifecycle|GetAllowStreamResponseToApplicationLifecycle)
        (if (Variables|ApplicationLifecycle|GetStreamStoppedDueToSuspend)
          (Fortnite|FortShippingLog self "ApplicationHasEnteredForeground: Opening stream due to application entering foreground")
          (CallFunction|OpenSpecialEventStream)
          (Variables|ApplicationLifecycle|SetStreamStoppedDueToSuspend false)
          (else
            (Fortnite|FortShippingLog self "ApplicationHasEnteredForeground: Opening stream due to application entering foreground")))
        (else
          (Fortnite|FortShippingLog self "ApplicationHasEnteredForeground: doing nothing - 
AllowStreamResponseToApplicationLifecycle is false"))))))

(event Custom|OpenSpecialEventStream
  (bind _videoplayer (Variables|Default|GetVideoPlayer))
  ;[0 0 0 0.5] Media Framework
  ;;[0 0 0 0.5] Open stream media source and start streaming
  (Networking|CanExecuteCosmeticEvents
    (:True
      (Fortnite|FortShippingLog self "OpenSpecialEventStream: called...")
      (if (and (Utilities|IsValid (Variables|Default|GetMediaSource) :!class /Script/Engine.KismetSystemLibrary) (Utilities|IsValid _videoplayer :!class /Script/Engine.KismetSystemLibrary))
        (if (not (CallFunction|IsMediaPlayerPlaying))
          (if (CallFunction|IsStreamUrlEmpty)
            ;;; If there's no Stream URL, we need to request it using the VUID
            (Fortnite|FortShippingLog self "OpenSpecialEventStream: requesting URL")
            (Variables|Default|SetAutoStartStream true)
            (CallFunction|TryLoadStreamURL)
            (else
              ;;; Otherwise, startup the stream like normal
              (if (Variables|Default|GetOpenRequestPending)
                (Fortnite|FortShippingLog self "OpenSpecialEventStream: not opening source because there is a currently pending open request")
                (else
                  (Variables|Default|SetOpenRequestPending true)
                  (bind _returnvalue (Media|MediaPlayer|OpenSource _videoplayer (Variables|Default|GetMediaSource)))
                  (Fortnite|FortShippingLog self (Utilities|String|SelectString "OpenSpecialEventStream: Stream media source opened successfully." "OpenSpecialEventStream: Stream media source failed to open." _returnvalue))))))
          (else
            (Fortnite|FortShippingLog self "OpenSpecialEventStream: Media Player is already playing")))
        (else
          ;;;[1 0 0 0.5] Error: Media Source or Player is invalid
          (Fortnite|FortShippingLog self "OpenSpecialEventStream: Media Source or Media Player is invalid!" true))))))

(event Custom|CloseSpecialEventStream
  ;[0 0 0 0.5] Media Framework
  ;;[0 0 0 0.5] Close stream media source
  (Networking|CanExecuteCosmeticEvents
    (:True
      (Fortnite|FortShippingLog self "CloseSpecialEventStream: called...")
      (if (CallFunction|IsMediaPlayerPlaying)
        (Fortnite|FortShippingLog self "CloseSpecialEventStream: closing the stream")
        (Media|MediaPlayer|Close (Variables|Default|GetVideoPlayer) :!class /Script/MediaAssets.MediaPlayer)
        (Actor|Tick|SetActorTickEnabled false)
        (else
          (Fortnite|FortShippingLog self "CloseSpecialEventStream: called when Media Player is not playing"))))))

(event Custom|SetAllowShareBlocking
  (params (NewValue :type Boolean))
  ;[0 0 0 0.5] Share Blocking
  (bind _output_get (Variables|ShareBlocking|SetAllowShareBlocking NewValue))
  ;; Unblock sharing if it is currently blocked and the cheat wants it unblocked
  (if (and (Variables|ShareBlocking|GetIsSharingBlocked) (not _output_get))
    (CallFunction|UnblockSharing)
    (elif (Math|Boolean|ANDBoolean (not (Variables|ShareBlocking|GetIsSharingBlocked)) _output_get (Media|MediaPlayer|IsPlaying (Variables|Default|GetVideoPlayer)))
      ;; Block sharing if it is currently not blocked and the cheat wants it blocked, but only when playing the video player
      (CallFunction|BlockSharing))))

(event Custom|OnMediaOpened
  (params (OpenedUrl :type String))
  ;[0 0 0 0.5] Media Framework
  ;;[0 0 0 0.5] On Media Opened
  (Networking|CanExecuteCosmeticEvents
    (:True
      (Fortnite|FortShippingLog self "OnMediaOpened: media opened successfully")
      (CallFunction|BlockSharing)
      (Default|CallOnVideoPlaybackStarted)
      (Variables|Default|SetOpenRequestPending false)
      (Variables|PlayRetry|SetPlayRetryCount 0)
      (Actor|Tick|SetActorTickEnabled true))))

(event Custom|OnMediaOpenedFailed
  (params (FailedUrl :type String))
  ;[0 0 0 0.5] Media Framework
  ;;[1 0.01578 0 0.5] On Media Open Failed
  (Networking|CanExecuteCosmeticEvents
    (:True
      (Variables|Default|SetOpenRequestPending false)
      (CallFunction|UnblockSharing)
      (Actor|Tick|SetActorTickEnabled false)
      (Fortnite|FortShippingLog self "OnMediaOpenedFailed: Failed to open media source, retrying..." true)
      (CallFunction|RetryOpenStream))))

(event Custom|OnMediaClosed
  (bind _videoplayer (Variables|Default|GetVideoPlayer))
  ;[0 0 0 0.5] Media Framework
  ;;[0 0 0 0.5] On Media Closed
  (Networking|CanExecuteCosmeticEvents
    (:True
      (Utilities|IsValid _videoplayer
        (:Is%20Valid
          (CallFunction|UnblockSharing)
          (if (Media|MediaPlayer|HasError _videoplayer)
            (Fortnite|FortShippingLog self "OnMediaClosed: Media closed due to an error, attempting a retry" true)
            (CallFunction|RetryOpenStream)
            (else
              (Fortnite|FortShippingLog self "OnMediaClosed: Media closed without error."))))
        (:Is%20Not%20Valid
          ;;;[1 0 0 0.5] Error: Video player invalid
          (Fortnite|FortShippingLog self "OnMediaClosed: VideoPlayer is not valid!" true))))))

(event Custom|RetryOpenStream
  (bind _returnvalue (CallFunction|CalculateCurrentPlayRetryInterval))
  ;[0 0 0 0.5] Media Framework
  ;;[0 0 0 0.5] Retry opening the stream
  (if (< (Variables|PlayRetry|GetPlayRetryCount) (Variables|PlayRetry|GetMaxPlayRetries))
    (Fortnite|FortShippingLog self (Utilities|String|ToString(Text) (Utilities|Text|FormatText "NSLOCTEXT(\"[EB26832506F51E826F44DB7912CE0A62]\", \"CE91B4AA4579F127B52AA281C3E32E84\", \"RetryOpenStream: attempting to retry opening the stream, attempt: {RetryCount}, interval: {RetryInterval}\")" (Variables|PlayRetry|GetPlayRetryCount) _returnvalue)) true)
    (Utilities|Time|SetTimerbyEvent (EventDispatchers|CreateEvent :!event Custom|OpenSpecialEventStream) _returnvalue false true)
    (Math|Integer|IncrementInt (Variables|PlayRetry|GetPlayRetryCount))
    (else
      ;;;[1 0 0 0.5] Error: Ran out of retries
      (Fortnite|FortShippingLog self "RetryOpenStream: Ran out of retry attempts!" true))))

(event Custom|ServerSetAllowShareBlocking
  (params (NewValue :type Boolean))
  ;[0 0 0 0.5] Share Blocking
  ;; Tell the Server and all Clients to update their AllowShareBlocking flag
  (Networking|SwitchHasAuthority
    (:Authority
      (CallFunction|SetAllowShareBlocking NewValue))))

(event Custom|OnMediaEndReached
  ;[0 0 0 0.5] Media Framework
  ;;[0 0 0 0.5] OnMediaEndReach
  (Networking|CanExecuteCosmeticEvents
    (:True
      (Fortnite|FortShippingLog self "OnMediaEndReached: video player reported that we have reached the end of the current playing media")
      (Default|CallOnVideoPlaybackEnded))))

(fn BindMediaOpenCloseEvents
  (bind _videoplayer (Variables|Default|GetVideoPlayer))
  (Utilities|IsValid (Variables|Default|GetVideoPlayer)
    (:Is%20Valid
      (Media|MediaPlayer|BindEventtoOnMediaOpened _videoplayer (EventDispatchers|CreateEvent :!event OnMediaOpened))
      (Media|MediaPlayer|BindEventtoOnMediaOpenFailed _videoplayer (EventDispatchers|CreateEvent :!event OnMediaOpenedFailed))
      (Media|MediaPlayer|BindEventtoOnMediaClosed _videoplayer (EventDispatchers|CreateEvent :!event OnMediaClosed))
      (Media|MediaPlayer|BindEventtoOnEndReached _videoplayer (EventDispatchers|CreateEvent :!event OnMediaEndReached)))
    (:Is%20Not%20Valid
      ;[1 0 0 0.5] Error: Media source not found
      (Fortnite|FortShippingLog self "BindMediaOpenCloseEvents: video player not found!" true))))

(fn BindURLRequestEvents
  (bind _mediasource (Variables|Default|GetMediaSource))
  (Utilities|IsValid _mediasource
    (:Is%20Valid
      ; Unbind from events just in case, to prevent double binding
      (Stream|UnbindEventfromOnVideoUrlSuccess _mediasource (EventDispatchers|CreateEvent :!event OnStreamURLRequestSuccess))
      (Stream|UnbindEventfromOnVideoUrlFailed _mediasource (EventDispatchers|CreateEvent :!event OnStreamURLRequestFailed))
      ; Rebind to the events now that we know we won't double bind
      (Stream|BindEventtoOnVideoUrlSuccess _mediasource (EventDispatchers|CreateEvent :!event OnStreamURLRequestSuccess))
      (Stream|BindEventtoOnVideoUrlFailed _mediasource (EventDispatchers|CreateEvent :!event OnStreamURLRequestFailed)))
    (:Is%20Not%20Valid
      ;[1 0 0 0.5] Error: Media source not found
      (Fortnite|FortShippingLog self "BindURLRequestEvents: media source not found!" true))))

(fn BlockSharing
  ; For testing purposes, allow disabling of shareblocking functionality
  (if (Variables|ShareBlocking|GetAllowShareBlocking)
    ; Make sure sharing isn't already blocked
    (if (Variables|ShareBlocking|GetIsSharingBlocked)
      (Fortnite|FortShippingLog self "DisableSharing: not activating shareblocking because sharing is already disabled" true)
      (else
        (Fortnite|FortShippingLog self "BlockSharing: Disabling sharing because it is currently enabled")
        ; Block sharing and track its current state
        (Player|DisableSharing (Variables|Default|GetMediaSource))
        (Variables|ShareBlocking|SetIsSharingBlocked true)))
    (else
      (Fortnite|FortShippingLog self "BlockSharing: not allowing shareblocking due to CanActivateShareblocking being false!" true))))

(fn CalculateCurrentPlayRetryInterval
  ; Exponential increase of the interval between retries
  (return (* (Variables|PlayRetry|GetPlayRetryInitialInterval) (Math|Float|Power (Variables|PlayRetry|GetPlayRetryIntervalExponentialBase) (Math|Conversions|ToFloat(Integer) (Variables|PlayRetry|GetPlayRetryCount))))))

(fn HandleStreamURLSuccess
  (params (URL :type String))
  (bind _mediasource (Variables|Default|GetMediaSource))
  (Networking|CanExecuteCosmeticEvents
    (:True
      (bind _output_get (Variables|/Script/MediaAssets.StreamMediaSource|SetStreamUrl URL _mediasource))
      (Stream|SetUrl _mediasource _output_get)
      ; Cycle the stream off/on if we're already running
      (if (CallFunction|IsMediaPlayerPlaying)
        (Fortnite|FortShippingLog self "OnStreamURLRequestSuccess: Cycling stream on/off")
        (CallFunction|CloseSpecialEventStream)
        (CallFunction|OpenSpecialEventStream)
        (elif (Variables|Default|GetAutoStartStream)
          ; Turn the stream on if we received an Open request while loading the URL
          (Fortnite|FortShippingLog self "OnStreamURLRequestSuccess:  Opening stream due to Autostart")
          (CallFunction|OpenSpecialEventStream)
          (else
            (Fortnite|FortShippingLog self "OnStreamURLRequestSuccess:  Stream not started - stream was not playing and autostart is false."))))
      ; Make sure to disable autostart
      (Variables|Default|SetAutoStartStream false))))

(fn IsVideoPlayerPlaying
  (return (and (Utilities|IsValid (Variables|Default|GetVideoPlayer) :!class /Script/Engine.KismetSystemLibrary) (Media|MediaPlayer|IsPlaying (Variables|Default|GetVideoPlayer)))))

(fn IsMediaPlayerPlaying
  (return (and (Utilities|IsValid (Variables|Default|GetVideoPlayer) :!class /Script/Engine.KismetSystemLibrary) (Media|MediaPlayer|IsPlaying (Variables|Default|GetVideoPlayer)))))

(fn IsOpenRequestPending
  (return (Variables|Default|GetOpenRequestPending)))

(fn IsStreamUrlEmpty
  (bind _mediasource (Variables|Default|GetMediaSource))
  (Utilities|IsValid _mediasource
    (:Is%20Valid
      (return (Utilities|String|IsEmpty (Media|MediaSource|GetUrl _mediasource))))
    (:Is%20Not%20Valid
      (return true))))

(fn OnStreamURLRequestFailed
  (params (RetrievedUrl :type String) (Code :type Integer) (RequestType :type EEpicMediaRequestType%20Enum :default "UNKNOWN"))
  (Fortnite|FortShippingLog self (Utilities|String|Append "OnStreamURLRequestFailed: Failed to receive valid URL (Code: " (Utilities|String|ToString(Integer) Code) ", RequestType: " (Utilities|String|EnumtoString RequestType) ")") true)
  ; Make sure to disable autostart if it was on
  (Variables|Default|SetAutoStartStream false))

(fn OnStreamURLRequestSuccess
  (params (RetrievedUrl :type String) (Code :type Integer) (RequestType :type EEpicMediaRequestType%20Enum :default "UNKNOWN"))
  (Networking|CanExecuteCosmeticEvents
    (:True
      (Fortnite|FortShippingLog self (Utilities|String|Append "OnStreamURLRequestSuccess:  Stream URL found successfully (Code: " (Utilities|String|ToString(Integer) Code) ", RequestType: " (Utilities|String|EnumtoString RequestType) ")"))
      (CallFunction|HandleStreamURLSuccess RetrievedUrl))))

(fn TryLoadStreamURL
  (bind _mediasource (Variables|Default|GetMediaSource))
  (Networking|CanExecuteCosmeticEvents
    (:True
      (CallFunction|BindURLRequestEvents)
      (Utilities|IsValid _mediasource
        (:Is%20Valid
          (if (not (Utilities|String|IsEmpty (Variables|/Script/EpicStreamMediaSource.EpicStreamMediaSource|GetStreamID _mediasource)))
            (Fortnite|FortShippingLog self (Utilities|String|Append "TryLoadStreamURL: Requesting Stream URL with VUID: " (Variables|/Script/EpicStreamMediaSource.EpicStreamMediaSource|GetStreamID _mediasource)))
            (Stream|RequestVideoUrl _mediasource 0)
            (else
              ;[1 0 0 0.5] We can't request a Stream URL without a VUID
              (Fortnite|FortShippingLog self "TryLoadStreamURL: Could not request Stream URL, VUID is invalid!" true))))
        (:Is%20Not%20Valid
          ;[1 0 0 0.5] Error: If Media Source is invalid we have a problem
          (Fortnite|FortShippingLog self "TryLoadStreamURL: Could not request Stream URL, Media Source is invalid!"))))))

(fn UnblockSharing
  ; Don't reinstate sharing if sharing is already allowed
  (if (Variables|ShareBlocking|GetIsSharingBlocked)
    (Fortnite|FortShippingLog self "UnblockSharing: Reinstating sharing because it is currently disabled")
    ; Allow sharing and tracks its current status
    (Player|ReinstateSharing (Variables|Default|GetMediaSource))
    (Variables|ShareBlocking|SetIsSharingBlocked false)
    (else
      (Fortnite|FortShippingLog self "UnblockSharing: not reinstating sharing because sharing is already allowed" true))))

(fn ConstructionScript)
